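# Translucent unlit Niagara FX material for the Chrome Anomaly nano-mist effect, usable on ribbons, sprites and mesh particles.
# Two noise layers distort a stretchable base shape, which combines with panning sand-texture layers to mask tinted, hue-shifted emissive and opacity, modulated by particle color.
# A curve atlas row along U shapes opacity and the width of optional SplineThicken vertex offset.
# A dynamic parameter offsets animation time.
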Material /BattlePassS22/QuestActors/Narrative_Pt1/FX/Materials/ChromeAnomaly/M_ChromePlayer_NanoMist_Anomaly
graph {
 properties {
  bEnableResponsiveAA=true
  bUsedWithNiagaraMeshParticles=true
  bUsedWithNiagaraRibbons=true
  bUsedWithNiagaraSprites=true
  blendMode=BLEND_Translucent
  pixelDepthOffsetMode=PDOM_Legacy
  shadingModel=MSM_Unlit
  usedShadingModels=Unlit
 }
 nodes [
  n1 = TextureObjectParameter() {
   group=Main
   parameterName=Main_Texture
   samplerType=SAMPLERTYPE_Normal
   sortPriority=0
   texture=/BattlePassS22/QuestActors/Narrative_Pt1/FX/Textures_ColorCurves/T_ChromePlayer_Sand1
  }
  n2 = VectorParameter() {group=Main, parameterName=Main_UVControl, channelNames {a=VPan, b=UPan, g=VScale, r=UScale}, defaultValue=<rgba 1 1 0 0>}
  n3 = NamedRerouteUsage()
  n4 = TextureCoordinate()
  n5 = MaterialFunctionCall(Float2=@n4) {materialFunction=&BreakOutFloat2Components}
  n6 = StaticSwitchParameter(
   True=MaterialFunctionCall(X=@n5.G, Y=@n5.G) {materialFunction=&MakeFloat2}.Result
   False=@n4
  ) {parameterName=SwitchUV}
  n7 = MaterialFunctionCall(Texture=@n1, UVControl_RGB=@n2, UVControl_A=@n2, "[Override]TimeInput"=@n3, "[Override]UVInput"=@n6) {materialFunction=&FXCommon_TextureSample}
  n8 = TextureObjectParameter() {
   group=BaseShape
   parameterName=BaseShape_Texture
   samplerType=SAMPLERTYPE_Grayscale
   sortPriority=0
   texture=/Game/Gadgets/Assets/VinderTech_GliderChute/Glider_Thrive_Female/FX/Texture/T_RibbonBaseShape4
  }
  n9 = VectorParameter() {group=BaseShape, parameterName=BaseShape_UVControl, channelNames {a=VPan, b=UPan, g=VScale, r=UScale}, defaultValue=<rgba 1 1 0 0>}
  n10 = TextureObjectParameter() {
   group=NoiseA
   parameterName=NoiseA_Texture
   samplerType=SAMPLERTYPE_LinearGrayscale
   sortPriority=0
   texture=/Game/Gadgets/Assets/VinderTech_GliderChute/Glider_Thrive_Female/FX/Texture/T_AntiMaterialCharge_Dissolve_01
  }
  n11 = VectorParameter() {group=NoiseA, parameterName=NoiseA_UVControl, channelNames {a=VPan, b=UPan, g=VScale, r=UScale}, defaultValue=<rgba 1 0.5 1 0>}
  n12 = VectorParameter() {
   group=NoiseA
   parameterName=NoiseA_NoiseControl
   sortPriority=60
   channelNames {a=VStrength, b=UStrength, g=VOffset, r=UOffset}
   defaultValue=<rgba -0.5 -0.5 0 -0.5>
  }
  n13 = MaterialFunctionCall!(
   NoiseValue=MaterialFunctionCall(Texture=@n10, UVControl_RGB=@n11, UVControl_A=@n11, "[Override]TimeInput"=NamedRerouteUsage(), "[Override]UVInput"=NamedRerouteUsage()) {materialFunction=&FXCommon_TextureSample}.RG
   Control_RGB=@n12
   Control_A=@n12
  ) {materialFunction=/Game/Effects/Niagara/MasterMaterials/Functions/Common/FXCommon_NoiseControl}
  n14 = VectorParameter() {group=NoiseB, parameterName=NoiseB_UVControl, channelNames {a=VPan, b=UPan, g=VScale, r=UScale}, defaultValue=<rgba 1 1 0 0>}
  n15 = VectorParameter() {
   group=NoiseB
   parameterName=NoiseB_NoiseControl
   sortPriority=60
   channelNames {a=VStrength, b=UStrength, g=VOffset, r=UOffset}
   defaultValue=<rgba -0.5 -0.5 0.01 0.01>
  }
  n16 = MaterialFunctionCall(
   NoiseValue=MaterialFunctionCall(
    Texture=TextureObjectParameter() {group=NoiseB, parameterName=NoiseB_Texture, sortPriority=0}
    UVControl_RGB=@n14
    UVControl_A=@n14
    "[Override]TimeInput"=NamedRerouteUsage()
    "[Override]UVInput"=NamedRerouteUsage()
   ) {materialFunction=&FXCommon_TextureSample}.RG
   Control_RGB=@n15
   Control_A=@n15
  ) {materialFunction=&FXCommon_NoiseControl}
  n17 = OneMinus(None=MaterialFunctionCall(Float2=@n6) {materialFunction=&BreakOutFloat2Components}.R)
  n18 = ScalarParameter() {defaultValue=0.6, group=03.BaseShape, parameterName=BaseShape_Stretch}
  n19 = AppendVector(A=Add(A=Multiply(A=@n17, B=@n18), B=ScalarParameter() {group=03.BaseShape, parameterName=BaseShape_StretchOffset}), B=Constant())
  n20 = MaterialFunctionCall(
   Texture=@n8
   UVControl_RGB=@n9
   UVControl_A=@n9
   "[Override]TimeInput"=NamedRerouteUsage()
   "[Override]UVInput"=Add(A=Add(A=@n13.NoiseValue, B=@n16.NoiseValue), B=Add(A=@n6, B=@n19))
  ) {materialFunction=&FXCommon_TextureSample}
  n21 = TextureObjectParameter() {
   group=Tri
   parameterName=Tri_Texture
   samplerType=SAMPLERTYPE_Normal
   sortPriority=0
   texture=&T_ChromePlayer_Sand1
  }
  n22 = VectorParameter() {group=Tri, parameterName=Tri_UVControl, channelNames {a=VPan, b=UPan, g=VScale, r=UScale}, defaultValue=<rgba 1 1 0 0>}
  n23 = Multiply(A=Add(A=MaterialFunctionCall(Texture=@n21, UVControl_RGB=@n22, UVControl_A=@n22, "[Override]TimeInput"=@n3, "[Override]UVInput"=@n6) {materialFunction=&FXCommon_TextureSample}.G)) {constB=0.5}
  n24 = ScalarParameter() {parameterName=TriMask2}
  n25 = Multiply(
   A=Power(Base=@n20.R, Exp=ScalarParameter() {parameterName=TriMask})
   B=MaterialFunctionCall(In=Power(Base=@n23, Exp=@n24), Contrast=ScalarParameter() {parameterName="TriMask2@"}) {materialFunction=&CheapContrast}.Result
  )
  n26 = VectorParameter() {parameterName=TriC, defaultValue=<rgba 0 0 0 1>}
  n27 = MaterialFunctionCall(
   Hue Shift Percentage=@n7.R
   Texture=Add(A=Multiply(A=@n25, B=@n26.RGB), B=VectorParameter() {parameterName=Color, defaultValue=<rgba 1000 0 0 1>}.RGB)
  ) {materialFunction=&HueShift}
  n28 = TextureObjectParameter() {
   group=Sec
   parameterName=Sec_Texture
   samplerType=SAMPLERTYPE_Normal
   sortPriority=0
   texture=&T_ChromePlayer_Sand1
  }
  n29 = VectorParameter() {group=Sec, parameterName=Sec_UVControl, channelNames {a=VPan, b=UPan, g=VScale, r=UScale}, defaultValue=<rgba 1 1 0 0>}
  n30 = Add(A=MaterialFunctionCall(Texture=@n28, UVControl_RGB=@n29, UVControl_A=@n29, "[Override]TimeInput"=@n3, "[Override]UVInput"=@n6) {materialFunction=&FXCommon_TextureSample}.G)
  n31 = Multiply(A=Multiply(A=Add(A=@n7.G)) {constB=0.5}, B=Add(A=Multiply(A=@n30) {constB=0.5}, B=ScalarParameter() {parameterName=aw}))
  n32 = ParticleColor()
  MaterialFunctionCall(Emissive=Multiply(A=Multiply(A=@n27.Result, B=@n31), B=@n32.RGB)) {materialFunction=&MF_FX_EmissiveAdjustments}
  n33 = CurveAtlasRowParameter(CurveTime=MaterialFunctionCall(Float2=TextureCoordinate()) {materialFunction=&BreakOutFloat2Components}.R) {
   atlas=/BattlePassS22/QuestActors/Narrative_Pt1/FX/Textures_ColorCurves/CLA_ChromePlayer1
   curve=/BattlePassS22/QuestActors/Narrative_Pt1/FX/Textures_ColorCurves/CL_ChromePlayerEnergyRibbon1
   parameterName=RibbonCurve
  }
  n34 = Saturate(None=@n31)
  n35 = Multiply(A=Add(A=Multiply(A=@n34, B=@n20.R), B=ScalarParameter()), B=ScalarParameter() {parameterName=TotalOpacityScale})
  n36 = Multiply(A=@n25, B=ScalarParameter() {parameterName=TriToOpacityStrength})
  Multiply(A=@n32.A, B=Multiply(A=@n33.G, B=Add(A=@n35, B=@n36)))
  n37 = MaterialFunctionCall!(
   UVInput=TextureCoordinate()
   TimeInput=NamedRerouteUsage()
   Control_RGB=VectorParameter() {parameterName=TotalUVControl, defaultValue=<rgba 1 1 0 0>}.RGB
   Control_A
  ) {materialFunction=/Game/Effects/Niagara/MasterMaterials/Functions/Common/FXCommon_QuickUVControl}
  n38 = MaterialFunctionCall(Float2=@n37.UV) {materialFunction=&BreakOutFloat2Components}
  NamedRerouteDeclaration(None=StaticSwitchParameter(
    True=MaterialFunctionCall(X=@n38.G, Y=@n38.G) {materialFunction=&MakeFloat2}.Result
    False=@n37.UV
   ) {parameterName=SwitchUV}) {name=RemapedUV, nodeColor=<rgba 0 1 0.3672 1>}
  NamedRerouteDeclaration(None=Add(A=Time(), B=DynamicParameter() {paramNames [R,Param2,Param3,Param4]}.R)) {nodeColor=<rgba 0.875 1 0 1>}
  n39 = Multiply(A=@n33.R, B=ScalarParameter() {defaultValue=50, parameterName=Width})
  StaticSwitchParameter(
   True=MaterialFunctionCall(WidthBase=@n39, WidthTip=@n39, Expand U or V UV Channel=StaticBool()) {materialFunction=&SplineThicken}.WorldPositionOffset
   False=Constant()
  ) {defaultValue=true, parameterName=UseSpineThicken}
 ]
}
